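# Physics asset for the ServeStreet head cosmetic on large male player characters, covering only the jaw bones.
# A large offset capsule on deform_jaw and a tiny capsule on deform_jaw_base form the two bodies.
# A single constraint joins them, allowing a limited swing on one axis with zero limits on the other swing and twist, so the jaw behaves like a hinge.
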PhysicsAsset /BRCosmetics/Characters/Player/Male/Large/Heads/M_LRG_ServeStreet_Head/Meshes/M_LRG_ServeStreet_Head_Physics
bodies {
 deform_jaw [{shapeType=Capsule, length=18.6817, radius=18.2158, center=<10 18.8796 0>, rotation=<rot 90 90 0>}]
 deform_jaw_base [{shapeType=Capsule, length=0.505, radius=0.505, shapeName=deform_jaw_base_capsule, rotation=<rot 0 -90 0>}]
}
constraints [{bone1Name=deform_jaw, bone2Name=deform_jaw_base, swing1LimitDegrees=20, swing2LimitDegrees=0, twistLimitDegrees=0}]
