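# Control Rig for the ServeStreet large male head cosmetic, procedurally animating its neck, left eyelids and jaw.
# The neck follows the head through aim constraints, with a spherical pose reader weighting a corrective offset.
# Targeting blends left eyelids from setup-cached bind poses into an offset pose via Kalman smoothing.
# Unless disabled, targeting, emoting or in frontend, moving or zero speed enables curve-remapped jaw harmonics, randomly engaged with smoothed weight.
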ControlRigBlueprint /BRCosmetics/Characters/Player/Male/Large/Heads/M_LRG_ServeStreet_Head/Meshes/M_LRG_ServeStreet_Head_CtrlRig
graphs [
 New Function_ContainedGraph {nodes [
   n1 = RigVMFunctionEntryNode()
   > RigVMFunctionReturnNode(Transform=&RigUnit_GetRelativeTransformForItem(Parent=(Type=Bone,Name="root"), bChildInitial=False, bParentInitial=False, Child=@n1.Item).RelativeTransform)
  ]}
 New Function_ContainedGraph {nodes [
   n1 = RigVMFunctionEntryNode()
   > /Script/ControlRig.RigUnit_SetRelativeTransformForItem(Parent=(Type=Bone,Name="root"), Weight=1, bParentInitial=False, bPropagateToChildren=True, Child=@n1.Item, Value=@n1.Transform) {
    is=struct RigVMUnitNode
    template="Set Relative Transform::Execute(in Child,in Parent,in bParentInitial,in Value,in Weight,in bPropagateToChildren)"
   }
   > RigVMFunctionReturnNode()
   &RigUnit_SetTransform(ExecutePin=(), Weight=1, bInitial=False, Item=@n1.Item, Value=@n1.Transform)
  ]}
 RigVMFunctionLibrary {nodes [
   Get Root Relative Transform = RigVMCollapseNode(ExecuteContext=(), Item=(Type=None,Name="None"))
   Set Root Relative Transform = RigVMCollapseNode(ExecuteContext=(), Item=(Type=Bone,Name="None"), Transform=(Rotation=(X=0,Y=0,Z=0,W=1),Translation=(X=0,Y=0,Z=0),Scale3D=(X=1,Y=1,Z=1)))
  ]}
 RigVMFunction_MathBoolOr_2_ContainedGraph {nodes [
   n1 = RigVMFunctionEntryNode()
   RigVMFunction_MathBoolOr = &RigVMFunction_MathBoolOr(A=@n1.A, B=@n1.B)
   RigVMFunctionReturnNode(Result=&RigVMFunction_MathBoolOr(B=@n1.C, A=@RigVMFunction_MathBoolOr.Result).Result)
  ]}
 RigVMFunction_Sequence_1_ContainedGraph {nodes [
   n1 = RigVMFunctionEntryNode()
   RigVMFunction_Sequence = &RigVMFunction_Sequence(ExecuteContext=@n1.ExecuteContext)
   n2 = &RigVMFunction_Sequence(ExecuteContext=@RigVMFunction_Sequence.B)
   RigVMFunctionReturnNode(A=@RigVMFunction_Sequence.A, B=@n2.A, C=@n2.B)
  ]}
 RigVMModel {nodes [
   RigUnit_BeginExecution = &RigUnit_BeginExecution()
   RigVMFunction_Sequence = RigVMAggregateNode(ExecuteContext=@RigUnit_BeginExecution.ExecutePin)
   n1 = &RigUnit_SphericalPoseReader(
    ActiveRegionScaleFactors=(PositiveWidth=0.5,NegativeWidth=4,PositiveHeight=4,NegativeHeight=4)
    ActiveRegionSize=0.1
    Debug=(bDrawDebug=true,bDraw2D=False,bDrawLocalAxes=False,DebugScale=25,DebugSegments=20,DebugThickness=0.25)
    DriverAxis=(X=1,Y=0,Z=0)
    DriverItem=(Type=Bone,Name="head")
    FalloffRegionScaleFactors=(PositiveWidth=1,NegativeWidth=1,PositiveHeight=1,NegativeHeight=1)
    FalloffSize=0.361
    FlipHeightScaling=False
    FlipWidthScaling=False
    OptionalParentItem=(Type=Bone,Name="spine_05")
    RotationOffset=(X=-90,Y=0,Z=25)
    ExecutePin=@RigVMFunction_Sequence.A
   )
   > &RigUnit_ParentConstraint(
    AdvancedSettings=(InterpolationType=Average,RotationOrderForFilter=XZY)
    Child=(Type=Bone,Name="cr_deform_neck_aim")
    Parents=((Item=(Type=Bone,Name="head"),Weight=1))
    Weight=1
   )
   > &RigUnit_AimConstraintLocalSpaceOffset(
    Child=(Type=Bone,Name="cr_deform_neck_limited")
    Filter=(bX=True,bY=True,bZ=True)
    Parents=((Item=(Type=Bone,Name="cr_deform_neck_aim"),Weight=1))
    Weight=1
    WorldUp=(Target=(X=0,Y=0,Z=1),Kind=Location,Space=(Type=Bone,Name="cr_deform_neck_aim"))
   )
   > &RigUnit_ModifyTransforms(
    ItemToModify=((Item=(Type=Bone,Name="cr_deform_neck_limited"),Transform=(Rotation=(X=0,Y=0,Z=0.2164,W=0.9763),Translation=(X=2,Y=1.5,Z=0),Scale3D=(X=1,Y=1,Z=1))))
    WeightMinimum=0
    Weight=@n1.OutputParam
   )
   > &RigUnit_ParentConstraint(
    AdvancedSettings=(InterpolationType=Average,RotationOrderForFilter=XZY)
    Child=(Type=Bone,Name="deform_neck")
    Parents=((Item=(Type=Bone,Name="cr_deform_neck_limited"),Weight=1))
    Weight=1
   )
   RigVMFunction_MathBoolToInteger = &RigVMFunction_MathBoolToInteger(Value=RigVMVariableNode() {variable=Is Targeting}.Value)
   KalmanFilter = &RigVMFunction_KalmanFloat(
    BufferSize=4
    Value=RigVMDispatchNode(Values=(0,1), Index=@RigVMFunction_MathBoolToInteger.Result) {template="DISPATCH_RigVMDispatch_SelectInt32(in Index,in Values,out Result)"}.Result
   ) {template="KalmanFilter::Execute(in Value,in BufferSize,out Result)"}
   &RigUnit_ModifyTransforms(
    ItemToModify="((Item=(Type=Bone,Name=\"deform_upper_eyelid_l\"),Transform=(Rotation=(X=0,Y=0,Z=0,W=1),Translation=(X=0,Y=0,Z=0),Scale3D=(X=1,Y=1,Z=1))),(Item=(Type=Bone,Name=\"deform_lower_eyelid_l\"),Transform=(Rotation=(X=0,Y=0,Z=0,W=1),Translation=(X=0,Y=0,Z=0),Scale3D=( ...(truncated)"
    Mode=OverrideLocal
    WeightMinimum=0
    Weight=@KalmanFilter.Result
    ItemToModify.0.Transform=RigVMVariableNode() {variable=Upper Eyelid L Bind Pose}.Value
    ItemToModify.1.Transform=RigVMVariableNode() {variable=Lower Eyelid L Bind Pose}.Value
    ExecutePin=@RigVMFunction_Sequence.B
   )
   > &RigUnit_ModifyTransforms(
    ItemToModify="((Item=(Type=Bone,Name=\"deform_upper_eyelid_l\"),Transform=(Rotation=(X=-0.7069,Y=0.0154,Z=-0.0154,W=-0.7069),Translation=(X=0.0118,Y=0,Z=-0.2698),Scale3D=(X=1,Y=1,Z=1))),(Item=(Type=Bone,Name=\"deform_lower_eyelid_l\"),Transform=(Rotation=(X=-0.7071,Y=0,Z=0, ...(truncated)"
    WeightMinimum=0
    Weight=@KalmanFilter.Result
   )
   n2 = &RigVMFunction_ControlFlowBranch(Condition=RigVMVariableNode() {variable=Disable Head Anim}.Value, ExecuteContext=@RigVMFunction_Sequence.C)
   RigVMFunction_MathBoolOr = RigVMAggregateNode(
    C=RigVMVariableNode() {variable=Frontend}.Value
    A=RigVMVariableNode() {variable=Is Targeting}.Value
    B=RigVMVariableNode() {variable=Is Playing Emote}.Value
   )
   RigVMFunction_MathBoolNot = &RigVMFunction_MathBoolNot(Value=@RigVMFunction_MathBoolOr.Result)
   n3 = RigVMVariableNode() {variable=Speed}
   RigVMFunction_MathBoolAnd = &RigVMFunction_MathBoolAnd(
    B=@RigVMFunction_MathBoolNot.Result
    A=&RigVMFunction_MathBoolOr(
     B=RigVMDispatchNode(B=0, A=@n3.Value) {template="DISPATCH_RigVMDispatch_CoreEquals(in A,in B,out Result)"}.Result
     A=RigVMVariableNode() {variable=Is Moving}.Value
    ).Result
   )
   n4 = &RigVMFunction_ControlFlowBranch(ExecuteContext=@n2.False, Condition=@RigVMFunction_MathBoolAnd.Result)
   LessEqual = /Script/RigVM.RigVMFunction_MathDoubleLessEqual(
    B=0
    A=/Script/RigVM.RigVMFunction_RandomFloat(
     Maximum=1
     Minimum=-6
     Seed=217
     Duration=&RigVMFunction_RandomFloat(Duration=6, Maximum=4, Minimum=1.2, Seed=217) {template="Random::Execute(in Seed,in Minimum,in Maximum,in Duration,out Result)"}.Result
    ) {is=struct RigVMUnitNode, template="Random::Execute(in Seed,in Minimum,in Maximum,in Duration,out Result)"}.Result
   ) {is=struct RigVMUnitNode, template="LessEqual::Execute(in A,in B,out Result)"}
   RigVMFunction_ControlFlowBranch = &RigVMFunction_ControlFlowBranch(Condition=@LessEqual.Result, ExecuteContext=@n4.True)
   /Script/ControlRig.RigUnit_ItemHarmonics(
    RotationOrder=YZX
    Targets=((Item=(Type=Bone,Name="harmonics_jaw"),Ratio=0.5))
    WaveAmplitude=(X=0,Y=0,Z=20)
    WaveEase=Linear
    WaveFrequency=(X=0,Y=0,Z=1)
    WaveMaximum=1
    WaveMinimum=0.5
    WaveNoise=(X=0,Y=0,Z=11)
    WaveOffset=(X=0,Y=0,Z=1)
    WaveSpeed=(X=0,Y=0,Z=1.05)
    ExecutePin=@RigVMFunction_ControlFlowBranch.Completed
   ) {is=struct RigVMUnitNode}
   RigUnit_GetTransform = &RigUnit_GetTransform(Item=(Type=Bone,Name="harmonics_jaw"), Space=LocalSpace, bInitial=False)
   RigVMFunction_MathQuaternionToEuler = &RigVMFunction_MathQuaternionToEuler(RotationOrder=ZYX, Value=@RigUnit_GetTransform.Transform.Rotation)
   RigVMFunction_AnimEvalRichCurve = &RigVMFunction_AnimEvalRichCurve(
    Curve="(EditorCurveData=(Keys=((InterpMode=RCIM_Linear,TangentMode=RCTM_Auto,TangentWeightMode=RCTWM_WeightedNone,Time=0,Value=0,ArriveTangent=0,ArriveTangentWeight=0,LeaveTangent=0,LeaveTangentWeight=0),(InterpMode=RCIM_Linear,TangentMode=RCTM_Auto,TangentWeight ...(truncated)"
    SourceMaximum=0
    SourceMinimum=-20
    TargetMaximum=0
    TargetMinimum=-20
    Value=@RigVMFunction_MathQuaternionToEuler.Result.Z
   )
   RigVMFunction_MathQuaternionFromEuler = &RigVMFunction_MathQuaternionFromEuler(
    Euler=(X=0,Y=0,Z=0)
    Euler.Z=@RigVMFunction_AnimEvalRichCurve.Result
    Euler.X=@RigVMFunction_MathQuaternionToEuler.Result.X
    Euler.Y=@RigVMFunction_MathQuaternionToEuler.Result.Y
   )
   > RigUnit_SetRotation = &RigUnit_SetRotation(
    Item=(Type=Bone,Name="deform_jaw")
    bInitial=False
    Weight=&RigVMFunction_KalmanFloat(BufferSize=8, Value=RigVMVariableNode() {variable=Engage}.Value) {template="KalmanFilter::Execute(in Value,in BufferSize,out Result)"}.Result
    Value=@RigVMFunction_MathQuaternionFromEuler.Result
   )
   RigVMVariableNode(Value=1, ExecuteContext=@RigVMFunction_ControlFlowBranch.False) {variable=Engage}
   RigVMVariableNode(Value=0, ExecuteContext=@RigVMFunction_ControlFlowBranch.True) {variable=Engage}
   &RigUnit_ParentConstraint(
    AdvancedSettings=(InterpolationType=Average,RotationOrderForFilter=XZY)
    Child=(Type=Bone,Name="cr_deform_neck_aim")
    ExecutePin=()
    Parents=((Item=(Type=Bone,Name="head"),Weight=0.7),(Item=(Type=Bone,Name="spine_05"),Weight=0.3))
    Weight=1
   )
  ]}
 RigVMModel Rig {nodes [
   n1 = &RigUnit_PrepareForExecution()
   RigUnit_GetRelativeTransformForItem = &RigUnit_GetRelativeTransformForItem(Child=(Type=Bone,Name="deform_upper_eyelid_l"), Parent=(Type=Bone,Name="deform_jaw"), bChildInitial=False, bParentInitial=false)
   RigVMVariableNode(ExecuteContext=@n1.ExecutePin, Value=@RigUnit_GetRelativeTransformForItem.RelativeTransform) {variable=Upper Eyelid L Bind Pose}
   n2 = &RigUnit_GetRelativeTransformForItem(Child=(Type=Bone,Name="deform_lower_eyelid_l"), Parent=(Type=Bone,Name="deform_jaw"), bChildInitial=False, bParentInitial=false)
   > RigVMVariableNode(Value=@n2.RelativeTransform) {variable=Lower Eyelid L Bind Pose}
  ]}
]
hierarchy [
 root {loc=<0 0 0>, rot=<0 0 0 1>}
 pelvis {parent=root, loc=<0 -0.7321 92.4218>, rot=<0.0127 0.707 -0.0127 -0.707>}
 spine_01 {parent=pelvis, loc=<5.6002 -0.5128 0>, rot=<0 0 0.034 -0.9994>}
 spine_02 {parent=spine_01, loc=<8.0756 0.0139 0>, rot=<0 0 0.0238 -0.9997>}
 spine_03 {parent=spine_02, loc=<7.0395 0 0>, rot=<0 0 0.0149 -0.9999>}
 spine_04 {parent=spine_03, loc=<8.1023 0 0>, rot=<0 0 -0.0132 -0.9999>}
 spine_05 {parent=spine_04, loc=<10.0141 0 0>, rot=<0 0 -0.0015 -1>}
 clavicle_l {parent=spine_05, loc=<2.5856 -0.2866 -4.3777>, rot=<0.0465 -0.7289 -0.009 -0.6829>}
 upperarm_l {parent=clavicle_l, loc=<18.2264 0.1342 -0.001>, rot=<0.0253 -0.2693 0.0013 -0.9627>}
 upperarm_twist_01_l {parent=upperarm_l, loc=<7.9796 0.0843 -0.6195>, rot=<0 0 0 -1>}
 neck_01 {parent=spine_05, loc=<10.815 1.9196 0>, rot=<0 0 -0.1864 -0.9825>}
 neck_02 {parent=neck_01, loc=<4.7347 0.3045 0>, rot=<0 0 0.0363 -0.9993>}
 head {parent=neck_02, loc=<5.7632 0.5552 0>, rot=<0 0 0.115 -0.9934>}
 faceAttach {parent=head, loc=<-4.146 -0.8545 0>, rot=<0.0029 -0.7071 -0.0029 -0.7071>}
 C_brow_mid {parent=faceAttach, loc=<0 8.1598 4.6104>, rot=<0 0 0 -1>}
 L_brow_mid {parent=faceAttach, loc=<2.8904 7.9296 5.3259>, rot=<-0.0001 -0.0042 0.2553 -0.9669>}
 L_brow_outer {parent=faceAttach, loc=<5.1605 5.8651 4.6582>, rot=<-0.0287 -0.04 0.4491 -0.8921>}
 L_cheek_inner {parent=faceAttach, loc=<3.7048 6.7635 -0.1198>, rot=<-0.0706 -0.2697 0.3184 -0.906>}
 C_lip_upper_mid {parent=faceAttach, loc=<0.0022 8.8861 -2.1487>, rot=<0.2225 0.0013 -0.0005 -0.9749>}
 L_lip_upper_outer {parent=faceAttach, loc=<1.3923 8.41 -2.1797>, rot=<0.0359 -0.0728 0.2627 -0.9615>}
 L_lip_corner {parent=faceAttach, loc=<2.3267 7.4634 -2.5915>, rot=<0.003 0.008 -0.0146 -0.9999>}
 C_nose_bridge {parent=faceAttach, loc=<0 9.8506 1.2827>, rot=<-0.2218 0 0 -0.9751>}
 R_brow_mid {parent=faceAttach, loc=<-2.8904 7.9296 5.3259>, rot=<-0.0001 0.0042 -0.2553 -0.9669>}
 R_brow_outer {parent=faceAttach, loc=<-5.1605 5.8651 4.6582>, rot=<-0.0287 0.04 -0.4491 -0.8921>}
 R_cheek_inner {parent=faceAttach, loc=<-3.7048 6.7635 -0.1198>, rot=<-0.0706 0.2697 -0.3184 -0.906>}
 R_lip_corner {parent=faceAttach, loc=<-2.3267 7.4634 -2.5915>, rot=<0.003 -0.008 0.0146 -0.9999>}
 C_jaw {parent=faceAttach, loc=<-0.0134 -0.1613 -1.4291>, rot=<0 0 0 1>}
 L_lip_lower_outer {parent=C_jaw, loc=<1.3888 8.3831 -1.4241>, rot=<0.013 0.0445 0.3057 -0.951>}
 C_lip_lower_mid {parent=C_jaw, loc=<0.0223 8.7898 -1.4818>, rot=<0 0 0 -1>}
 R_lip_lower_outer {parent=C_jaw, loc=<-1.3619 8.3831 -1.4241>, rot=<0.013 -0.0445 -0.3057 -0.951>}
 teeth_lower {parent=C_jaw, loc=<0.0134 6.0835 -1.7637>, rot=<0 0 0 1>}
 R_eye {parent=faceAttach, loc=<8.8557 17.2208 7.2946>, rot=<0 0 0.9998 -0.0175>}
 R_eye_lid_upper_mid {parent=R_eye, loc=<0 0 0>, rot=<0.0523 0 0 -0.9986>}
 R_eye_lid_lower_mid {parent=R_eye, loc=<0 0 0>, rot=<0.0329 0 0 -0.9995>}
 L_eye {parent=faceAttach, loc=<-8.8672 7.4021 7.3567>, rot=<0 0 0.9998 0.0175>}
 L_eye_lid_upper_mid {parent=L_eye, loc=<0 0 0>, rot=<0.0523 0 0 -0.9986>}
 L_eye_lid_lower_mid {parent=L_eye, loc=<0 0 0>, rot=<0.0329 0 0 -0.9995>}
 R_lip_upper_outer {parent=faceAttach, loc=<-1.3923 8.41 -2.1797>, rot=<0.0359 0.0728 -0.2627 -0.9615>}
 teeth_upper {parent=faceAttach, loc=<0 6.5028 -1.1408>, rot=<0 0 0 1>}
 cr_deform_neck_aim {parent=head, loc=<20 0 0>, rot=<0 0 0 1>}
 clavicle_r {parent=spine_05, loc=<2.5855 -0.2866 4.3777>, rot=<0.7289 0.0465 -0.6829 0.009>}
 upperarm_r {parent=clavicle_r, loc=<-18.2264 -0.1342 0.001>, rot=<0.0253 -0.2693 0.0013 -0.9627>}
 upperarm_twist_01_r {parent=upperarm_r, loc=<-7.9796 -0.0843 0.6195>, rot=<0 0 0 -1>}
 deform_neck {parent=spine_05, loc=<7.0143 2.3756 0>, rot=<0 0 -0.0401 -0.9992>}
 deform_jaw {parent=deform_neck, loc=<8.2272 -18.996 0>, rot=<0 0 0 1>}
 deform_eye_l {parent=deform_jaw, loc=<8.6047 30.3294 -8.9226>, rot=<0.7071 0 0.7071 0>}
 deform_pupil_l {parent=deform_eye_l, loc=<0.0549 -7.4697 0.0177>, rot=<0 0 0 1>}
 deform_eye_r {parent=deform_jaw, loc=<8.6047 30.3294 8.9227>, rot=<0 0.7071 0 -0.7071>}
 deform_pupil_r {parent=deform_eye_r, loc=<-0.0549 7.4697 -0.018>, rot=<0 0 0 1>}
 deform_upper_eyelid_l {parent=deform_jaw, loc=<8.6047 30.3294 -8.9226>, rot=<0.6733 0.2162 0.6733 0.2162>}
 deform_lower_eyelid_l {parent=deform_jaw, loc=<8.6047 30.3294 -8.9226>, rot=<0.6733 0.2162 0.6733 0.2162>}
 deform_upper_eyelid_r {parent=deform_jaw, loc=<8.6047 30.3295 8.9227>, rot=<-0.2162 0.6733 0.2162 -0.6733>}
 deform_lower_eyelid_r {parent=deform_jaw, loc=<8.6047 30.3294 8.9227>, rot=<-0.2162 0.6733 0.2162 -0.6733>}
 harmonics_jaw {parent=deform_neck, loc=<8.2272 -18.996 0>, rot=<0 0 0 1>}
 cr_deform_neck_full {parent=spine_05, loc=<7.0143 2.3756 0>, rot=<0 0 -0.0401 -0.9992>}
 cr_deform_neck_limited {parent=spine_05, loc=<7.0143 2.3756 0>, rot=<0 0 -0.0401 -0.9992>}
 attach_cape {parent=spine_05, loc=<9.017 -11.714 0>, rot=<0.0616 -0.7044 -0.0616 -0.7044>}
 attach_backpack {parent=spine_05, loc=<-0.123 -13.473 0>, rot=<-0.0093 -0.707 0.0093 -0.707>}
]
variables [
 Disable Head Anim {type=bool, default=False}
 Engage {type=float, default=0}
 Frontend {type=bool, default=False}
 Is Moving {type=bool, default=False}
 Is Playing Emote {type=bool, default=False}
 Is Targeting {type=bool, default=False}
 Lower Eyelid L Bind Pose {type=FTransform, default=(Rotation=(X=0,Y=0,Z=0,W=1),Translation=(X=0,Y=0,Z=0),Scale3D=(X=1,Y=1,Z=1))}
 Speed {type=double, default=0}
 Upper Eyelid L Bind Pose {type=FTransform, default=(Rotation=(X=0,Y=0,Z=0,W=1),Translation=(X=0,Y=0,Z=0),Scale3D=(X=1,Y=1,Z=1))}
]
