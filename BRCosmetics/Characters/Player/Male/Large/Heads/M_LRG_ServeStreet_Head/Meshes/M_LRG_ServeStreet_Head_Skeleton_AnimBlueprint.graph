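# Animation Blueprint for the ServeStreet large male player head cosmetic, driving that head's Control Rig from the input pose.
# The rig receives targeting, movement, speed and emote state plus two flags computed on each animation update.
# Default head yap is disabled when no valid pawn owner exists.
# In the front end, the owner mesh's ServeStreet_FrontendIdle curve sets the custom-idle flag, passed as the rig's Frontend input.
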AnimBlueprint /BRCosmetics/Characters/Player/Male/Large/Heads/M_LRG_ServeStreet_Head/Meshes/M_LRG_ServeStreet_Head_Skeleton_AnimBlueprint
target_skeleton=&/M_LRG_ServeStreet_Head_Skeleton
graphs [
 AnimGraph {nodes [
   n1 = Control Rig!(
    Source=Input Pose()
    Alpha
    Is Targeting:bool=Get bIsTargeting()
    Is Moving:bool=Get bIsMoving2D()
    Speed:double=Get PawnSpeed2D()
    Is Playing Emote:bool=Get bIsPlayingEmote()
    Frontend:bool=Get Custom Idle Playing() -> (Custom Idle Playing:bool)
    Disable Head Anim:bool=Get Disable Default Head Yap() -> (Disable Default Head Yap:bool)
   ) -> (Pose) {
    alphaBoolBlend {blendOption=Linear}
    controlRigAssetReference {blueprintRigClass=&/M_LRG_ServeStreet_Head_CtrlRig.M_LRG_ServeStreet_Head_CtrlRig_C}
   }
   Output Pose(Result=@n1)
   Get bIsInFrontEnd()
  ]}
 EventGraph {nodes [
   Event BlueprintUpdateAnimation()
   > n1 = Sequence()
   > n2 = Is Valid(InputObject=TryGetPawnOwner())
   Set Disable Default Head Yap(execute=@n2.Is Not Valid, Disable Default Head Yap:bool=true) -> (Output_Get:bool)
   Set Disable Default Head Yap(execute=@n2.Is Valid, Disable Default Head Yap=false)
   Branch(execute=@n1.then_1, Condition=Get bIsInFrontEnd())
   n3 = GetAnimInstance(self=GetOwnerSkeletalMeshComponent())
   > Set Custom Idle Playing(Custom Idle Playing:bool="Nearly Equal (Float)"(A=GetCurveValue(self=@n3, CurveName=ServeStreet_FrontendIdle), B=1, ErrorTolerance=0.01)) -> (Output_Get:bool)
   TryGetPawnOwner()
  ]}
]
variables [
 Custom Idle Playing {type=boolean}
 Disable Default Head Yap {type=boolean}
]
